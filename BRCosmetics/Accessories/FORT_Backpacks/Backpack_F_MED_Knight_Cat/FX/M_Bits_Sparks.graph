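# Translucent unlit Niagara sprite material for spark bits in the Knight Cat backpack cosmetic's effects.
# Sprite_Capsule takes particle size X as diameter, Y as velocity-axis scale, and the normalized pivot-to-particle vector as direction, orienting sparks radially.
# Particle color feeds MF_FX_EmissiveAdjustments and, in a separate branch, multiplies the bits_spark texture; the graph shows no links from these to material outputs.
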Material /BRCosmetics/Accessories/FORT_Backpacks/Backpack_F_MED_Knight_Cat/FX/M_Bits_Sparks
graph {
 properties {bUsedWithNiagaraSprites=true, blendMode=BLEND_Translucent, pixelDepthOffsetMode=PDOM_Legacy, shadingModel=MSM_Unlit, usedShadingModels=Unlit}
 nodes [
  n1 = ParticlePositionWS()
  n2 = MaterialFunctionCall() {materialFunction=&ObjectPivotPoint}
  n3 = ParticleSize()
  MaterialFunctionCall!(
   Particle Position
   Normalized Particle Direction=Normalize(VectorInput=Subtract(A=@n1, B=@n2.Object Pivot Location))
   Capsule Diameter=ComponentMask(None=@n3) {r=true}
   Capsule Pivot Alignment
   Scale Along Velocity Vector=ComponentMask(None=@n3) {g=true}
   ---------Scale Over Velocity Parameters---------
   Particle Speed
   Minimum Speed
   Max Speed
   Minimum Scale Factor
   Maximum Scale Factor
   ---------- Advanced -----------
   Cap Size Multiplier
   Use Camera Offset
   Use Advanced Features
   World Position to Negate
  ) {materialFunction=/Engine/Functions/Engine_MaterialFunctions02/WorldPositionOffset/Sprite_Capsule}
  n4 = ParticleColor()
  MaterialFunctionCall(Emissive=@n4.RGB) {materialFunction=&MF_FX_EmissiveAdjustments}
  Multiply(A=@n4.RGB, B=TextureSample() {texture=&/bits_spark}.RGB)
 ]
}
