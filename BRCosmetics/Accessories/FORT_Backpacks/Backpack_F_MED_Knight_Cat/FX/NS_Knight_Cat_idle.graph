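# Looping idle effect for the Knight Cat backpack cosmetic, tinted by variant color user parameters.
# Flare, Flare2 and Glow are looping sprites with pulsing alpha; both flares follow the skeletal mesh.
# Pink_Bits spawns three mesh-attached sprites, each bobbing on its own ID-selected curve scaled by MotionBits.
# A second renderer gets darker, enlarged outline values, while brightness adapts to time of day and front end.
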NiagaraSystem /BRCosmetics/Accessories/FORT_Backpacks/Backpack_F_MED_Knight_Cat/FX/NS_Knight_Cat_idle
dependencies {
 usedDataInterfaces [
  &NiagaraDataInterfaceCurve
  &NiagaraDataInterfaceVectorCurve
  &NiagaraDataInterfaceVector2DCurve
 ]
 usedDynamicInputs []
 usedModules [&SystemState]
}
emitters {
 Flare {
  emitterUpdateScript [
   &EmitterState(
    Inactive Response=::NewEnumerator1
    Life Cycle Mode=::NewEnumerator1
    Loop Behavior=::NewEnumerator1
    Loop Duration Mode=::NewEnumerator1
    UseLoopDelay=false
   )
   &SpawnBurst_Instantaneous(Use Spawn Probability=false, UseLoopCountLimit=false)
  ]
  particleSpawnScript [
   &InitializeParticle(Color Mode=::NewEnumerator0, Lifetime=1.5, Position Offset=<0 -2.2 0>, UsePositionOffset=true, Write Material Random=true)
   /SetVariables(
    Particles.FlareSize01:Vector2f=&Vector2DFromFloat(Value=75)
    Particles.FlareSize02:Vector2f=&Vector2DFromFloat(Value=50)
   ) {isSetParameters=true}
   /SetVariables(Particles.FlareCameraOffset01=5, Particles.FlareCameraOffset02=8) {isSetParameters=true}
   /SetVariables(
    Particles.FlareDynamic01:Vector4f=&MakeVector4(X=0.8, Y=1, Z=1, W=1)
    Particles.FlareDynamic02:Vector4f=&MakeVector4(X=1, Y=0.2, Z=0.5, W=1)
   ) {isSetParameters=true}
  ]
  particleUpdateScript [
   &ParticleState(
    Kill Particles When Lifetime Has Elapsed=false
    Let Infinitely Lived Particles Die When Emitter Deactivates=true
    Loop Particles Lifetime=true
   )
   &SkeletalMeshLocation(Bone Sampling Mode=::NewEnumerator4, MaskSpawnGroup=false, Skeletal Mesh=$User.Skeletal Mesh)
   /SetVariables(Particles.AlphaCurve:NiagaraFloat=&FloatFromCurve(FloatCurve=DataInterface() {propertyValues {Curve {keys [<at 0 0.5 RCIM_Cubic>,<at 0.6 1 RCIM_Cubic>,<at 1 0.5 RCIM_Cubic>]}}})) {isSetParameters=true}
   /SetVariables(
    Particles.FlareColor01:LinearColor=&MultiplyLinearColor(
     Linear Color=$User.Variant_Color_Light
     Scale Factor=&MakeFloat4FromVectorAndFloat("Vector (RGB)"=<0.25 0.25 0.25>, "Float (Alpha)"=$Particles.AlphaCurve)
    )
    Particles.FlareColor02:LinearColor=&MultiplyLinearColor(
     Linear Color=$User.Variant_Color_Light
     Scale Factor=&MakeFloat4FromVectorAndFloat("Vector (RGB)"=<0.75 0.75 0.75>, "Float (Alpha)"=$Particles.AlphaCurve)
    )
   ) {isSetParameters=true}
  ]
  rendererClasses [&NiagaraSpriteRendererProperties]
  renderers [
   {rendererClass=&NiagaraSpriteRendererProperties}
   {rendererClass=&NiagaraSpriteRendererProperties, rendererIndex=1}
  ]
 }
 Flare2 {
  emitterUpdateScript [
   &EmitterState(
    Inactive Response=::NewEnumerator1
    Life Cycle Mode=::NewEnumerator1
    Loop Behavior=::NewEnumerator1
    Loop Duration Mode=::NewEnumerator1
    UseLoopDelay=false
   )
   &SpawnBurst_Instantaneous(Use Spawn Probability=false, UseLoopCountLimit=false)
  ]
  particleSpawnScript [
   &InitializeParticle(
    Color=&MultiplyLinearColor(Linear Color=$User.Variant_Color_Light, Scale Factor=&MakeVector4_FromFloat(Float=0.2))
    Lifetime=1.5
    Sprite Size Mode=::NewEnumerator3
    Uniform Sprite Size=22
    UsePositionOffset=false
    Write Material Random=true
   )
   &DynamicMaterialParameters(
    Index 0 Param 1=0.8
    Index 0 Param 2=1
    Index 0 Param 3=1
    Index 0 Param 4=1
    Module.Param0WriteEnabled=true
    Module.Param1WriteEnabled=true
    Module.Param2WriteEnabled=true
    Module.Param3WriteEnabled=true
   )
  ]
  particleUpdateScript [
   &CameraOffset(Camera Offset Amount=11)
   &ParticleState(
    Kill Particles When Lifetime Has Elapsed=false
    Let Infinitely Lived Particles Die When Emitter Deactivates=true
    Loop Particles Lifetime=true
   )
   &SkeletalMeshLocation(
    Bone Sampling Mode=::NewEnumerator4
    MaskSpawnGroup=false
    Sampled Position Offset Coordinate Space:ENiagaraSkelMeshTransforms=::NewEnumerator0
    Skeletal Mesh=$User.Skeletal Mesh
   )
   &ScaleColor(
    Scale Alpha=&FloatFromCurve(
     FloatCurve=DataInterface() {propertyValues {Curve {keys [<at 0 0.5 RCIM_Cubic>,<at 0.6 1 RCIM_Cubic>,<at 1 0.5 RCIM_Cubic>]}}}
     Scale Curve=5
    )
    ScaleA=true
    ScaleRGB=true
   )
  ]
  rendererClasses [&NiagaraSpriteRendererProperties]
  renderers [{rendererClass=&NiagaraSpriteRendererProperties}]
 }
 Glow {
  emitterUpdateScript [
   &EmitterState(Life Cycle Mode=::NewEnumerator1, Loop Behavior=::NewEnumerator1, Loop Duration Mode=::NewEnumerator1, UseLoopDelay=false)
   &SpawnBurst_Instantaneous(Use Spawn Probability=false, UseLoopCountLimit=false)
  ]
  particleSpawnScript [
   &InitializeParticle(
    Color=&MultiplyLinearColor(Linear Color=$User.Variant_Color_Light, Scale Factor=<0.25 0.25 0.25 0.25>)
    Lifetime=2
    Sprite Size Mode=::NewEnumerator3
    Uniform Sprite Size=50
    UsePositionOffset=false
    Write Material Random=true
   )
   &CameraOffset(Camera Offset Amount=10)
   &DynamicMaterialParameters(
    Index 0 Param 1=3
    Index 0 Param 2=1.5
    Index 0 Param 3=1
    Index 0 Param 4=1
    Module.Param0WriteEnabled=true
    Module.Param1WriteEnabled=true
    Module.Param2WriteEnabled=true
    Module.Param3WriteEnabled=true
   )
  ]
  particleUpdateScript [
   &ParticleState(
    Kill Particles When Lifetime Has Elapsed=false
    Let Infinitely Lived Particles Die When Emitter Deactivates=true
    Loop Particles Lifetime=true
   )
   &ScaleColor(
    Scale Alpha=&FloatFromCurve(FloatCurve=DataInterface() {propertyValues {Curve {keys [
         {interpMode=RCIM_Cubic, tangentMode=RCTM_User, time=0, value=0.5}
         {interpMode=RCIM_Cubic, tangentMode=RCTM_User, time=0.5, value=1}
         {tangentMode=RCTM_User, time=1, value=0.5}
        ]}}})
    ScaleA=true
    ScaleRGB=true
   )
  ]
  rendererClasses [&NiagaraSpriteRendererProperties]
  renderers [{rendererClass=&NiagaraSpriteRendererProperties}]
 }
 Pink_Bits {
  emitterUpdateScript [
   &EmitterState(Life Cycle Mode=::NewEnumerator1, Loop Duration=0, Recalculate Duration Each Loop=true, UseLoopDelay=false)
   &SpawnBurst_Instantaneous(
    Spawn Count=3
    Spawn Probability=HlslExpression() {hlslExpression="Emitter.LoopCount == 0"}
    Use Spawn Probability=true
    UseLoopCountLimit=false
   )
  ]
  particleSpawnScript [
   &InitializeParticle(
    Color=&MultiplyLinearColor(
     Linear Color=$User.Variant_Color
     Scale Factor=&MakeFloat4FromVectorAndFloat("Vector (RGB)"=&VectorFromFloat(Value=$User.ColorScale), "Float (Alpha)"=1)
    )
    Lifetime Min=1.75
    Lifetime Mode=::NewEnumerator1
    Position Mode=::NewEnumerator0
    Sprite Size=<13 37>
    Sprite Size Mode=::NewEnumerator4
    Write Material Random=true
   )
   &ShapeLocation(
    MaskSpawnGroup=false
    Offset=<0 -4 0>
    Offset Mode=::NewEnumerator0
    Override Seed=false
    Ring / Disc Distribution Mode:ENiagaraSphereDistributionMode=::NewEnumerator1
    Ring Radius=15
    Rotation Angle=0.75
    Rotation Axis=<0 1 0>
    Rotation Mode=::NewEnumerator1
    Shape Primitive=::NewEnumerator4
    U Position=&ReturnNormalizedExecIndex(Normalized Index Includes One=false)
    UseOwnerScale=true
   )
   &SpriteFacingAndAlignment(
    Sprite Alignment=&RotateVectorByQuaternion(
     Vector To Rotate=<1 -0.1 0>
     Quaternion=&FindQuatBetweenTwoVectors(From Vector=<0 0 1>, To Vector=$Particles.ShapeLocation.ShapeBitangent)
    )
    Sprite Facing=<0 0 0>
    WriteAlignment=false
    WriteFacing=true
   )
  ]
  particleUpdateScript [
   &ParticleState(
    Kill Particles When Lifetime Has Elapsed=false
    Let Infinitely Lived Particles Die When Emitter Deactivates=true
    Loop Particles Lifetime=true
   )
   &SkeletalMeshLocation(
    MaskSpawnGroup=false
    Mesh Sampling Type=::NewEnumerator5
    Position Sampling=::NewEnumerator2
    Skeletal Mesh=$User.Skeletal Mesh
    Socket Sampling Mode=::NewEnumerator5
   )
   /SetVariables(Particles.NewPosi:Vector3f=&Add_Vector(
     A=Linked() {linkedVariable {name=Particles.Position, type=NiagaraPosition}}
     B=&MakeCustomVectorFromBool(
      Bool=&self:ScratchDynamicInput_01(WantedID:NiagaraInt32=0)
      True Vector=&VectorFromCurve(
       VectorCurve=DataInterface() {propertyValues {
         XCurve {keys [<at 0 0>,<at 1 0>]}
         YCurve {keys [<at 0 0>,<at 1 0>]}
         ZCurve {keys [
           {interpMode=RCIM_Cubic, tangentMode=RCTM_User, time=0, value=-1}
           {interpMode=RCIM_Cubic, tangentMode=RCTM_User, time=0.5, value=1}
           {tangentMode=RCTM_User, time=1, value=-1}
          ]}
        }}
       Scale Curve=&Multiply_VectorByFloat(Vector=<1 1 1>, Float=$User.MotionBits)
      )
      False Vector=&MakeCustomVectorFromBool(
       Bool=&self:ScratchDynamicInput_01(WantedID:NiagaraInt32=1)
       True Vector=&VectorFromCurve(
        VectorCurve=DataInterface() {propertyValues {
          XCurve {keys [
            {interpMode=RCIM_Cubic, tangentMode=RCTM_User, time=0, value=-1}
            {interpMode=RCIM_Cubic, tangentMode=RCTM_User, time=0.5, value=1}
            {tangentMode=RCTM_User, time=1, value=-1}
           ]}
          YCurve {keys [<at 0 0>,<at 1 0>]}
          ZCurve {keys [
            {interpMode=RCIM_Cubic, tangentMode=RCTM_User, time=0, value=-1}
            {interpMode=RCIM_Cubic, tangentMode=RCTM_User, time=0.5, value=1}
            {tangentMode=RCTM_User, time=1, value=-1}
           ]}
         }}
        Scale Curve=&Multiply_VectorByFloat(Vector=<-1 1 1>, Float=$User.MotionBits)
       )
       False Vector=&MakeCustomVectorFromBool(
        Bool=&self:ScratchDynamicInput_01(WantedID:NiagaraInt32=2)
        True Vector=&VectorFromCurve(
         VectorCurve=DataInterface() {propertyValues {
           XCurve {keys [
             {interpMode=RCIM_Cubic, tangentMode=RCTM_User, time=0, value=-1}
             {interpMode=RCIM_Cubic, tangentMode=RCTM_User, time=0.5, value=1}
             {tangentMode=RCTM_User, time=1, value=-1}
            ]}
           YCurve {keys [<at 0 0>,<at 1 0>]}
           ZCurve {keys [
             {interpMode=RCIM_Cubic, tangentMode=RCTM_User, time=0, value=-1}
             {interpMode=RCIM_Cubic, tangentMode=RCTM_User, time=0.5, value=1}
             {tangentMode=RCTM_User, time=1, value=-1}
            ]}
          }}
         Scale Curve=&Multiply_VectorByFloat(Vector=<1 1 1>, Float=$User.MotionBits)
        )
        False Vector=<0 0 0>
       )
      )
     )
    )) {isSetParameters=true}
   &CameraOffset(Camera Offset Amount=5)
   &SolveForcesAndVelocity(Clamp Velocity=false, Limit Acceleration=false)
   &ScaleColor(
    Scale RGB=&VectorFromFloat(Value=&FN_FrontEndFloatModifier(
      Gameplay=&FNTimeOfDayScalar(Night Scale=0.5, Morning Scale=0.3, Evening Scale=0.3)
      Front End=1.5
     ))
    ScaleA=false
    ScaleRGB=true
   ) {module=ScaleColor001}
   &ScaleSpriteSize(
    Uniform Curve Scale=0.333
    Uniform Curve Sprite Scale=DataInterface() {propertyValues {Curve {keys [
        {arriveTangent=0.7018, interpMode=RCIM_Cubic, leaveTangent=0.7018, tangentMode=RCTM_User, time=0, value=0.25}
        {arriveTangent=-0.1134, interpMode=RCIM_Cubic, leaveTangent=-0.1134, tangentMode=RCTM_User, time=0.4, value=0.35}
        {arriveTangent=-0.0405, interpMode=RCIM_Cubic, leaveTangent=-0.0405, tangentMode=RCTM_User, time=0.75, value=0.25}
       ]}}}
   )
   /SetVariables(
    Particles.OutlineCameraOffset:NiagaraFloat=&Subtract_Float(A=$StackContext.CameraOffset, B=0.5)
    Particles.OutlineColor:LinearColor=&MultiplyLinearColor(
     Linear Color=$StackContext.Color
     Scale Factor=&MakeFloat4FromVectorAndFloat("Vector (RGB)"=&VectorFromFloat(Value=0.2), "Float (Alpha)"=1)
    )
    Particles.OutlineSize:Vector2f=/Niagara/DynamicInputs/Add/Add_Vector2(
     A:Vector2f=$StackContext.SpriteSize
     B:Vector2f=&Vector2DFromFloat(Value=&FN_FrontEndFloatModifier(Gameplay=0.45, Front End=0.333))
    )
   ) {isSetParameters=true}
  ]
  rendererClasses [&NiagaraSpriteRendererProperties]
  renderers [
   {rendererClass=&NiagaraSpriteRendererProperties}
   {rendererClass=&NiagaraSpriteRendererProperties, rendererIndex=1}
  ]
 }
}
summary {
 systemName=NS_Knight_Cat_idle
 emitters [
  {emitterName=Flare, rendererClasses [&NiagaraSpriteRendererProperties]}
  {emitterName=Glow, rendererClasses [&NiagaraSpriteRendererProperties]}
  {emitterName=Flare2, rendererClasses [&NiagaraSpriteRendererProperties]}
  {emitterName=Pink_Bits, rendererClasses [&NiagaraSpriteRendererProperties]}
 ]
 userVariables [
  User.bIsFrontEnd {type=NiagaraBool}
  User.bIsFrontEndPreview {type=NiagaraBool}
  User.ColorScale {type=NiagaraFloat, defaultValue=1, description=""}
  User.MotionBits {type=NiagaraFloat, defaultValue=0.75, description=""}
  User.Skeletal Mesh {
   type=NiagaraDataInterfaceSkeletalMesh
   defaultValue {
    struct=NiagaraExt_VariableValue_DataInterface
    value {dataInterface=&self:NiagaraDataInterfaceSkeletalMesh_0}
   }
  }
  User.Variant_Color {type=LinearColor, description="", defaultValue {struct=LinearColor, value=<rgba 1 0 0.5 1>}}
  User.Variant_Color_Light {type=LinearColor, description="", defaultValue {struct=LinearColor, value=<rgba 1 0 0.5 1>}}
 ]
}
