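# Animation Blueprint for the Knight Cat backpack that makes its dyn_crystal bone float and sway on the pack.
# Anim Dynamics adds damped spring motion, then additive bone edits turn the pack around and offset pack and crystal.
# Each update, sines of real time set variables driving a half-alpha crystal translation, roll and yaw.
# Skydiving variables and a stub Skydive function exist but never feed the pose.
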AnimBlueprint /BRCosmetics/Accessories/FORT_Backpacks/Backpack_F_MED_Knight_Cat/Meshes/F_MED_Knight_Cat_Pack_AnimBP
target_skeleton=&/F_MED_Knight_Cat_Pack_Skeleton
graphs [
 AnimGraph {nodes [
   n1 = Anim Dynamics - Bone: dyn_crystal(ComponentPose=Local To Component(LocalPose=Local Space Ref Pose()), Alpha=0.7, ExternalForce="0,0,0") {
    angularDampingOverride=0.9
    angularSpringConstant=150
    bGravityOverrideInSimSpace=true
    bLinearSpring=true
    bOverrideAngularDamping=true
    bOverrideLinearDamping=true
    gravityScale=0
    linearDampingOverride=0.995
    linearSpringConstant=50
    alphaBoolBlend {blendOption=Linear}
    boundBone {boneName=dyn_crystal}
    componentLinearAccScale=<0.3 0.3 0.3>
    physicsBodyDefinitions [{
      boundBone {boneName=dyn_crystal}
      boxExtents=<25 25 25>
      constraintSetup {twistAxis=AxisZ, angularLimitsMax=<0 0 0>, angularLimitsMin=<0 0 0>, angularTarget=<0 0 0>, linearAxesMax=<2 0 0>, linearAxesMin=<-1.5 0 0>}
      localJointOffset=<0 0 0>
     }]
   }
   n2 = "Transform (Modify) Bone - Bone: pack"(ComponentPose=@n1, Translation="0,-4,0", Rotation="0,180,0") {
    rotationMode=BMM_Additive
    translationMode=BMM_Additive
    alphaBoolBlend {blendOption=Linear}
    boneToModify {boneName=pack}
    rotation {yaw=180}
    translation {y=-4}
   }
   n3 = "Transform (Modify) Bone - Bone: dyn_crystal"(ComponentPose=@n2, Alpha=1, Translation="0,0.5,0", Rotation="0,0,0", Scale="1,1,1") {
    rotationMode=BMM_Additive
    translationMode=BMM_Additive
    alphaBoolBlend {blendOption=Linear}
    boneToModify {boneName=dyn_crystal}
    translation {y=0.5}
   }
   n4 = Get float_crystal_X_multiply() -> (float_crystal_X_multiply:double)
   n5 = "Transform (Modify) Bone - Bone: dyn_crystal"!(
    ComponentPose=@n3
    Alpha=0.5
    Translation_X:double=Get float for sine X() -> (float for sine X:double)
    Translation_Y:double=Get float for sine Y() -> (float for sine Y:double)
    Translation_Z:double=0
    Rotation_Roll:float=@n4
    Rotation_Pitch:float=0
    Rotation_Yaw:float=@n4
    Scale="1,1,1"
   ) -> (Pose) {
    rotationMode=BMM_Additive
    rotationSpace=BCS_BoneSpace
    translationMode=BMM_Additive
    translationSpace=BCS_BoneSpace
    alphaBoolBlend {blendOption=Linear}
    boneToModify {boneName=dyn_crystal}
   }
   Output Pose(Result=Component To Local(ComponentPose=@n5)) {layerGroup=None}
   "To Float (Boolean)"(InBool=false)
  ]}
 EventGraph {nodes [
   Event BlueprintUpdateAnimation()
   n1 = GetRealTimeSeconds()
   n2 = "Sin (Radians)"(A=float * float(A=2, B=@n1) {is=title K2Node_PromotableOperator})
   > Set float for sine X(float for sine X:double=float - float(A=@n2, B=0.5)) -> (Output_Get:double)
   > Set float_crystal_X_multiply(float_crystal_X_multiply:double=float * float(A=@n2, B=1)) -> (Output_Get:double)
   > Set float for sine Y(float for sine Y:double="Sin (Radians)"(A=@n1)) -> (Output_Get:double)
   TryGetPawnOwner()
  ]}
 NewFunction_0 {nodes [
   NewFunction_0() {is=title K2Node_FunctionEntry}
   > Skydive!(execute, bFromBus:bool=false, InPawn:Fort Player Pawn Object Reference, GliderOverride:Athena Glider Item Definition Object Reference) -> (then, ReturnValue:bool)
  ]}
]
variables [
 crystal speed {type=number, default=1}
 float for sine X {type=number}
 float for sine Y {type=number}
 Float For Sine Y_0 {type=number}
 float_crystal_X_multiply {type=number}
 Is Skydive Floating {type=boolean}
 Is Skydive Floating_0 {type=boolean}
 Pawn Velocity {type=Vector}
 skydiving {type=number}
 Skydiving Alpha {type=number}
 Skydiving External Force Input {
  type=FortAnimInput_SkydivingExternalForce
  default {
   bApplyNoiseInActorSpace=false
   bUseNoisyClothGravity=false
   bUseSkydivingVectorForce=false
   perlinRangedOutMaxX=-5000
   perlinRangedOutMaxY=-5000
   perlinRangedOutMaxZ=200
   perlinRangedOutMinX=-20000
   perlinRangedOutMinY=-20000
   perlinRangedOutMinZ=100
   divingAdditive=<0 0 2000>
   divingMultiplier=<10800 10800 12000>
   floatingAdditive=<0 0 2000>
   floatingMultiplier=<4500 4500 4500>
   parachutingAdditive=<0 0 0>
   parachutingMultiplier=<-300 -300 -300>
  }
 }
]
